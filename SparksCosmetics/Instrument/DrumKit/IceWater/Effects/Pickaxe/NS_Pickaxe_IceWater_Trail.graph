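# Niagara swing trail for the IceWater pickaxe cosmetic, leaving icy smoke, snow and blue ribbon streaks.
# Sprite emitters IceSmoke and Snow spawn by distance moved on a user skeletal mesh socket, inherit capped velocity and fall under drag and gravity, with curl noise on the snow.
# Two anim-trail ribbons add a soft, lingering streak and a short-lived bright edge.
# Time-of-day functions tint and fade the smoke and both ribbons.
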NiagaraSystem /SparksCosmetics/Instrument/DrumKit/IceWater/Effects/Pickaxe/NS_Pickaxe_IceWater_Trail
dependencies {
 usedDataInterfaces [&NiagaraDataInterfaceCurve, &NiagaraDataInterfaceArrayColor]
 usedDynamicInputs []
 usedModules [&SystemState]
}
emitters {
 IceSmoke {
  emitterUpdateScript [
   &EmitterState(Life Cycle Mode=::NewEnumerator1, Loop Duration=0.01, UseLoopDelay=false)
   &SpawnPerUnit(Module.Use Max Movement Threshold=true, Module.Use Movement Tolerance=false, Spawn Spacing=15, Use Spawn Probability=false)
   &SpawnRate(Spawn Probability=0.8, SpawnRate=3, Use Spawn Probability=true) {enabled=false}
  ]
  particleSpawnScript [
   &InitializeParticle(
    Color=<rgba 0.6562 0.8155 1 0.05>
    Lifetime Max=1.5
    Lifetime Min=0.35
    Lifetime Mode=::NewEnumerator1
    Mass Mode=::NewEnumerator3
    Position Offset=<0 0 20>
    Sprite Rotation Mode=::NewEnumerator2
    Sprite Size Mode=::NewEnumerator1
    Uniform Sprite Size Max=65
    Uniform Sprite Size Min=50
    UsePositionOffset=false
    Write Material Random=true
   ) {module=InitializeParticle001}
   &SkeletalMeshLocation(
    Bone / Socket Index=2
    Inherit Velocity=false
    MaskSpawnGroup=false
    Mesh Sampling Type=::NewEnumerator5
    Skeletal Mesh=$User.SkeletalMesh
    Socket Sampling Mode=::NewEnumerator5
   )
   &ShapeLocation(MaskSpawnGroup=false, Override Seed=false, Shape Origin=$Particles.Position, Sphere Radius=3, UseOwnerScale=true)
   /SetVariables(Particles.Rotation:NiagaraFloat=&RandomRangeFloat(Minimum=-80, Maximum=80)) {isSetParameters=true, module=SetVariables_#1}
   &InheritVelocity(Module.Limit Speed=true, Velocity Limit=200)
  ]
  particleUpdateScript [
   &Drag(Drag=5, Use Linear Drag=true, Use Rotational Drag=false)
   /Niagara/Modules/Update/SubUV/SubUVAnimation(Number Of Frames=63)
   &GravityForce(Gravity=<0 0 -75>)
   &SolveForcesAndVelocity(Clamp Velocity=false, Limit Acceleration=false)
   &ParticleState()
   &ScaleSpriteSize(
    Scale Sprite Size Mode=::NewEnumerator0
    Uniform Scale Factor=&FloatFromCurve(FloatCurve=DataInterface() {propertyValues {Curve {keys [
         {arriveTangent=-0.0977, interpMode=RCIM_Cubic, leaveTangent=-0.0977, tangentMode=RCTM_User, time=0, value=0}
         {arriveTangent=1.2916, interpMode=RCIM_Cubic, leaveTangent=1.2916, tangentMode=RCTM_User, time=0.3, value=0.6}
         <at 1.0489 1 RCIM_Cubic>
        ]}}})
   )
   &ScaleColor(
    Scale Alpha=&Multiply_Float(
     A=&FloatFromCurve(FloatCurve=DataInterface() {propertyValues {Curve {keys [
          <at 0 0 RCIM_Cubic>
          {arriveTangent=0.257, interpMode=RCIM_Cubic, leaveTangent=0.257, tangentMode=RCTM_User, time=0.2, value=0.65}
          {arriveTangent=0.1612, interpMode=RCIM_Cubic, leaveTangent=0.1612, tangentMode=RCTM_User, time=1.0011, value=0}
         ]}}})
     B=&FNTimeOfDayScalar(Night Scale=1, Morning Scale=0.35, Day Scale=0.25, Evening Scale=0.25)
    )
    Scale RGB=&FN_Select_Vector_from_Time_Of_Day(Morning=<0.9 0.9 0.9>, Day=<1 1 1>, Evening=<1 1 1>, Night=<0.5 0.5 1>)
    ScaleA=true
    ScaleRGB=true
   ) {module=ScaleColor002}
  ]
  rendererClasses [&NiagaraSpriteRendererProperties]
  renderers [{rendererClass=&NiagaraSpriteRendererProperties}]
 }
 Snow {
  emitterUpdateScript [
   &EmitterState(Life Cycle Mode=::NewEnumerator1, Loop Duration=0.01, UseLoopDelay=false)
   &SpawnPerUnit(Module.Use Max Movement Threshold=true, Module.Use Movement Tolerance=false, Spawn Spacing=15, Use Spawn Probability=false)
   &SpawnRate(Spawn Probability=0.8, SpawnRate=3, Use Spawn Probability=true) {enabled=false}
  ]
  particleSpawnScript [
   &InitializeParticle(
    Color=&SelectLinearColorFromArray(
     Array Sampling Mode=::NewEnumerator0
     Direct Array Index=?
     Color Selection Array=DataInterface() {propertyValues {ColorData [<rgba 0.7104 1.0588 1.1 1>, <rgba 0.7949 0.9257 1.1 1>]}}
     Random Seed=0
    )
    Lifetime Max=1.25
    Lifetime Min=0.75
    Lifetime Mode=::NewEnumerator1
    Mass Mode=::NewEnumerator3
    Position Offset=<0 0 20>
    Sprite Rotation Mode=::NewEnumerator2
    Sprite Size Mode=::NewEnumerator1
    Uniform Sprite Size Max=15
    Uniform Sprite Size Min=10
    UsePositionOffset=false
    Write Material Random=true
   ) {module=InitializeParticle001}
   &SkeletalMeshLocation(
    Bone / Socket Index=2
    Inherit Velocity=false
    MaskSpawnGroup=false
    Mesh Sampling Type=::NewEnumerator5
    Skeletal Mesh=$User.SkeletalMesh
    Socket Sampling Mode=::NewEnumerator5
   )
   &ShapeLocation(MaskSpawnGroup=false, Override Seed=false, Shape Origin=$Particles.Position, Sphere Radius=10, UseOwnerScale=true)
   /SetVariables(Particles.Rotation:NiagaraFloat=&RandomRangeFloat(Minimum=-80, Maximum=80)) {isSetParameters=true, module=SetVariables_#1}
   &InheritVelocity(Module.Limit Speed=true, Velocity Limit=200)
  ]
  particleUpdateScript [
   &Drag(Drag=5, Use Linear Drag=true, Use Rotational Drag=false)
   &GravityForce(Gravity=<0 0 -75>)
   &CurlNoiseForce(Bias Noise=false, Noise Strength=&RandomRangeFloat(Minimum=25, Maximum=50), Pan Noise=false)
   &SolveForcesAndVelocity(Clamp Velocity=false, Limit Acceleration=false)
   &ParticleState()
   &ScaleSpriteSize(
    Scale Sprite Size Mode=::NewEnumerator0
    Uniform Scale Factor=&FloatFromCurve(
     FloatCurve=DataInterface() {propertyValues {Curve {keys [
         {arriveTangent=2.635, interpMode=RCIM_Cubic, leaveTangent=2.635, tangentMode=RCTM_User, time=0, value=0}
         <at 0.5 0.45 RCIM_Cubic>
         <at 1.0489 0 RCIM_Cubic>
        ]}}}
     Scale Curve=0.5
    )
   )
  ]
  rendererClasses [&NiagaraSpriteRendererProperties]
  renderers [{rendererClass=&NiagaraSpriteRendererProperties}]
 }
 SnowSmokeTrail {
  emitterUpdateScript [
   &EmitterState(Life Cycle Mode=::NewEnumerator1, Loop Behavior=::NewEnumerator1, UseLoopDelay=false)
   &FN_SpawnAnimTrail_v2()
  ]
  particleSpawnScript [
   /Game/Effects/Niagara/Modules/AnimTrail/FN_InitializeAnimTrail/FN_InitializeAnimTrailBoneIndecies(First Bone or Socket Index=1)
   /Game/Effects/Niagara/Modules/AnimTrail/FN_InitializeAnimTrail/FN_InitializeAnimTrail(AnimTrailWidth=0.5, Color=<rgba 0 0.5263 0.9635 0.95>, Lifetime=1.5)
   &SystemLocation(MaskSpawnGroup=false)
  ]
  particleUpdateScript [
   &ParticleState()
   &ScaleColor(
    Scale Alpha=&Multiply_Float(
     A=&FloatFromCurve(FloatCurve=DataInterface() {propertyValues {Curve {keys [
          {arriveTangent=6.1048, interpMode=RCIM_Cubic, leaveTangent=6.1048, tangentMode=RCTM_User, time=0, value=0}
          <at 0.333 1 RCIM_Cubic>
          <at 1 0 RCIM_Cubic>
         ]}}})
     B=&FNTimeOfDayScalar(Night Scale=0.5, Morning Scale=0.65, Day Scale=0.7, Evening Scale=0.7)
    )
    Scale RGB=&FN_Select_Vector_from_Time_Of_Day(Morning=<0.9 0.9 0.9>, Day=<1 1 1>, Evening=<1 1 1>, Night=<0.5 0.5 1>)
    ScaleA=true
    ScaleRGB=true
   ) {module=ScaleColor002}
   &ScaleColor(
    Scale Alpha=&FloatFromCurve(FloatCurve=DataInterface() {propertyValues {Curve {keys [
         {arriveTangent=6.1048, interpMode=RCIM_Cubic, leaveTangent=6.1048, tangentMode=RCTM_User, time=0, value=0}
         <at 0.333 1 RCIM_Cubic>
         <at 1 0 RCIM_Cubic>
        ]}}})
    Scale RGB=<1 1.6 2>
    ScaleA=true
    ScaleRGB=true
   ) {enabled=false}
   &ScaleRibbonWidth(Ribbon Width Scale=&FloatFromCurve(FloatCurve=DataInterface() {propertyValues {Curve {keys [{arriveTangent=-6.0141, interpMode=RCIM_Cubic, leaveTangent=-6.0141, tangentMode=RCTM_User, time=0, value=2}, <at 1 0 RCIM_Cubic>]}}}))
   &DynamicMaterialParameters(
    Index 0 Param 1=4
    Index 0 Param 2=1
    Index 0 Param 3=1
    Index 0 Param 4=&FloatFromCurve(FloatCurve=DataInterface() {propertyValues {Curve {keys [{arriveTangent=-2.1594, interpMode=RCIM_Cubic, leaveTangent=-2.1594, tangentMode=RCTM_User, time=0, value=1}, <at 1 0>]}}})
    Module.Param0WriteEnabled=true
    Module.Param1WriteEnabled=true
    Module.Param2WriteEnabled=true
    Module.Param3WriteEnabled=true
   )
   &FN_UpdateAnimTrail_v2()
  ]
  rendererClasses [&NiagaraRibbonRendererProperties]
  renderers [{rendererClass=&NiagaraRibbonRendererProperties}]
 }
 TrailHardEdge {
  emitterUpdateScript [
   &EmitterState(Life Cycle Mode=::NewEnumerator1, Loop Behavior=::NewEnumerator1, Loop Duration=0.22, UseLoopDelay=true)
   &FN_SpawnAnimTrail_v2()
  ]
  particleSpawnScript [
   &FN_InitializeAnimTrailBoneIndecies(First Bone or Socket Index=1)
   &FN_InitializeAnimTrail(AnimTrailWidth=1, Color=<rgba 0.75 1.2737 2 1>, Lifetime=0.15)
   &SystemLocation(MaskSpawnGroup=false)
  ]
  particleUpdateScript [
   &ScaleRibbonWidth(Ribbon Width Scale=&FloatFromCurve(FloatCurve=DataInterface() {propertyValues {Curve {keys [
         {arriveTangent=-1.2579, interpMode=RCIM_Cubic, time=0, value=1}
         {arriveTangent=-0.9664, interpMode=RCIM_Cubic, leaveTangent=-0.9664, tangentMode=RCTM_User, time=1, value=0}
        ]}}}))
   &ScaleColor(
    Scale Alpha=&Multiply_Float(
     A=&FloatFromCurve(FloatCurve=DataInterface() {propertyValues {Curve {keys [
          {interpMode=RCIM_Cubic, leaveTangent=-2.3646, tangentMode=RCTM_Break, time=0, value=1}
          <at 0.6689 0.0233>
          {arriveTangent=-0.2139, interpMode=RCIM_Cubic, tangentMode=RCTM_Break, time=1, value=0}
         ]}}})
     B=&FNTimeOfDayScalar(Night Scale=0.5, Morning Scale=0.65, Day Scale=0.7, Evening Scale=0.7)
    )
    Scale RGB=&FN_Select_Vector_from_Time_Of_Day(Morning=<0.9 0.9 0.9>, Day=<1 1 1>, Evening=<1 1 1>, Night=<0.5 0.5 1>)
    ScaleA=true
    ScaleRGB=true
   ) {module=ScaleColor002}
   &FN_UpdateAnimTrail_v2()
   &CameraOffset(Camera Offset Amount=10)
  ]
  rendererClasses [&NiagaraRibbonRendererProperties]
  renderers [{rendererClass=&NiagaraRibbonRendererProperties}]
 }
}
summary {
 systemName=NS_Pickaxe_IceWater_Trail
 emitters [
  {emitterName=SnowSmokeTrail, rendererClasses [&NiagaraRibbonRendererProperties]}
  {emitterName=Snow, rendererClasses [&NiagaraSpriteRendererProperties]}
  {emitterName=IceSmoke, rendererClasses [&NiagaraSpriteRendererProperties]}
  {emitterName=TrailHardEdge, rendererClasses [&NiagaraRibbonRendererProperties]}
 ]
 userVariables [
  User.RibbonIDOffset {type=NiagaraInt32}
  User.SkeletalMesh {
   type=NiagaraDataInterfaceSkeletalMesh
   defaultValue {
    struct=NiagaraExt_VariableValue_DataInterface
    value {dataInterface=&self:NiagaraDataInterfaceSkeletalMesh_1}
   }
  }
  User.TrailColor {type=LinearColor, description="", defaultValue {struct=LinearColor, value=<rgba 1 1 1 0.25>}}
 ]
}
